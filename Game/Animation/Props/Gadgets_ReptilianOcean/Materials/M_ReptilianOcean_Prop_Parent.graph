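# Masked parent material for the Reptilian Ocean emote gadget prop, layering a pulsing heated glow between the shared character PreFX and PostFX functions.
# A local-height gradient and mask texture drive a heat curve atlas, revealed along pre-skinned X by EffectWipe.
# The heat alters base color, emissive and roughness, while local height also clips opacity.
# Colored wipe-band and animated scale-spark masks are also built, though their final color blend appears unconnected.
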Material /Game/Animation/Props/Gadgets_ReptilianOcean/Materials/M_ReptilianOcean_Prop_Parent
graph {
 properties {
  bUseMaterialAttributes=true
  bUsedWithMeshDeformer=true
  bUsedWithNanite=true
  bUsedWithSkeletalMesh=true
  blendMode=BLEND_Masked
  pixelDepthOffsetMode=PDOM_Legacy
  previewMesh=/Game/Animation/Props/Gadgets_ReptilianOcean/Meshes/ReptilianOcean_Gadget
  shadingModel=MSM_FromMaterialExpression
  usedShadingModels=Default Lit | Subsurface | Clear Coat | Subsurface Profile | Cloth
 }
 nodes [
  n1 = ScalarParameter() {defaultValue=0.5, parameterName=EffectWipe}
  Divide(A=@n1) {constB=0.5}
  n2 = NamedRerouteDeclaration(None=VectorParameter() {parameterName=EffectColor, defaultValue=<rgba 2.4185 0.2031 3 1>}.RGB) {name=EffectColor, nodeColor=<rgba 0 0.5547 1 1>}
  n3 = PreSkinnedPosition()
  n4 = ComponentMask(None=ShaderStageSwitch(PixelShader=VertexInterpolator(VS=@n3).PS, VertexShader=@n3)) {r=true}
  n5 = Add(A=Divide(A=@n4) {constB=2000}) {constB=0.25}
  n6 = Saturate(None=Add(A=Multiply(A=@n5) {constB=2}, B=Fresnel() {baseReflectFraction=0, exponent=3}))
  n7 = OneMinus(None=@n6)
  n8 = Time()
  n9 = TextureSample(UVs=Add(A=TextureCoordinate() {uTiling=2, vTiling=2}, B=Multiply(A=@n8) {constB=0.2})) {texture=/Game/Animation/Props/Gadgets_ReptilianOcean/Textures/T_ReptilianOcean_Prop_basecolor}
  n10 = Multiply(A=Time()) {constB=-1}
  n11 = Multiply(A=Sine(None=Add(A=@n9.R, B=@n10))) {constB=0.5}
  n12 = Saturate(None=Subtract(A=@n6, B=Multiply(A=@n7, B=@n11)))
  LinearInterpolate(
   A=NamedRerouteUsage()
   B=LinearInterpolate(A=Multiply(A=@n2) {constB=0.5}, B=@n2, Alpha=Multiply(A=@n12) {constB=3})
   Alpha=NamedRerouteUsage()
  )
  n13 = MaterialFunctionCall() {materialFunction=&MF_CharacterParent_2023_PreFX}
  MaterialFunctionCall(Material=NamedRerouteUsage(), M=@n13.M, Pre Skinned Local Position=@n13.M, Distance Blend=@n13.M, Fresnel=@n13.M) {materialFunction=&MF_CharacterParent_2023_PostFX}
  n14 = ComponentMask(None=TextureSample() {samplerType=SAMPLERTYPE_Masks, texture=/Game/Animation/Props/Gadgets_ReptilianOcean/Textures/T_Emote_Prop_ReptilianOcean_S}.B) {b=true}
  Multiply(A=@n14) {constB=-1}
  NamedRerouteDeclaration(None=@n13.Result) {name=inMaterialAttributes, nodeColor=<rgba 1 0 0.6797 1>}
  n15 = ComponentMask(None=MaterialFunctionCall() {materialFunction=&LocalPosition}.Local Position) {b=true}
  n16 = OneMinus(None=Divide(A=@n15) {constB=100})
  n17 = SmoothStep(Value=Saturate(None=Add(A=@n16) {constB=-0.55}))
  n18 = Saturate(None=Multiply(A=@n17, B=OneMinus(None=@n14)))
  n19 = CurveAtlasRowParameter(CurveTime=LinearInterpolate(A=Multiply(A=@n18), B=@n17, Alpha=@n17) {constAlpha=0.2, constB=0}) {
   atlas=/Game/Animation/Game/MainPlayer/Emotes/ReptilianOcean/FX/CA_ReptilianOcean_Prop_FX
   curve=/Game/Animation/Game/MainPlayer/Emotes/ReptilianOcean/FX/C_ReptilianOcean_Prop_FX_Heated
  }
  n20 = Time()
  n21 = ConstantBiasScale(None=Sine(None=Multiply(A=@n20)))
  n22 = Multiply(A=Multiply(A=@n17, B=LinearInterpolate(Alpha=@n21) {constA=1, constB=1.2})) {constB=3}
  n23 = PreSkinnedPosition()
  n24 = ShaderStageSwitch(PixelShader=VertexInterpolator(VS=@n23).PS, VertexShader=@n23)
  n25 = Add(A=Divide(A=ComponentMask(None=@n24) {r=true}) {constB=1000}) {constB=-0.4}
  n26 = Multiply(A=Multiply(A=@n19, B=@n22), B=SmoothStep(Value=Add(A=@n25, B=@n1)) {constMax=0.5, constMin=0.1})
  NamedRerouteDeclaration(None=Add(A=NamedRerouteUsage(), B=@n26)) {name=OutEmissive, nodeColor=<rgba 1 0.9375 0 1>}
  n27 = NamedRerouteUsage()
  n28 = Divide(A=@n4) {constB=1000}
  n29 = ScalarParameter() {defaultValue=0.5, parameterName=EffectWipe}
  n30 = Sine(None=Add(A=Add(A=@n28) {constB=-0.25}, B=@n29))
  n31 = Add(
   A=SmoothStep(Min=LinearInterpolate(Alpha=@n27) {constA=0.1}, Value=SmoothStep(Value=@n30) {constMin=0.85})
   B=SmoothStep(Value=@n30) {constMin=0.97}
  )
  n32 = ScalarParameter() {defaultValue=4, parameterName=EffectTipIntensity}
  n33 = Add(A=@n5, B=Divide(A=@n29))
  n34 = Sine(None=@n33)
  NamedRerouteDeclaration(None=Add(A=Multiply(A=@n31, B=@n32) {constB=3}, B=SmoothStep(Value=@n34) {constMax=0.1})) {name=EffectMask, nodeColor=<rgba 0 1 0.0391 1>}
  NamedRerouteDeclaration(None=@n13.M) {name=inMask, nodeColor=<rgba 0.6875 1 0 1>}
  n35 = GetMaterialAttributes(None=NamedRerouteUsage()) {attributeGetTypes [#1, #2, #3]}
  NamedRerouteDeclaration(None=@n35.Base Color) {name=InBaseColor, nodeColor=<rgba 0.2188 0 1 1>}
  n36 = SetMaterialAttributes!(
   MaterialAttributes=@n35.MaterialAttributes
   Base Color=NamedRerouteUsage()
   Emissive Color=NamedRerouteUsage()
   Roughness=LinearInterpolate(A=@n35.Roughness, Alpha=Saturate(None=@n22))
   Opacity Mask=Saturate(None=Add(A=@n15) {constB=15})
  ) {attributeSetTypes [
    #1
    #2
    #3
    #4
   ]}
  NamedRerouteDeclaration(None=@n36) {name=outMaterialAttributes, nodeColor=<rgba 0.5938 0 1 1>}
  NamedRerouteDeclaration(None=@n35.Emissive Color) {name=InEmissive, nodeColor=<rgba 0.6875 1 0 1>}
  n37 = Time()
  n38 = TextureCoordinate() {uTiling=5, vTiling=5}
  n39 = Sine(None=Add(
    A=Multiply(A=@n37) {constB=0.2}
    B=TextureSample(UVs=@n38) {texture=&T_ReptilianOcean_Prop_basecolor}.R
   ))
  NamedRerouteDeclaration(None=Saturate(None=ConstantBiasScale(None=@n39))) {name=ScalesSpark, nodeColor=<rgba 0.0312 1 0 1>}
  NamedRerouteDeclaration(None=LinearInterpolate(A=NamedRerouteUsage(), B=@n26, Alpha=@n26)) {name=OutBaseColor, nodeColor=<rgba 1 0 0.2812 1>}
  NamedRerouteUsage()
  OneMinus(None=@n33)
 ]
}
